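; Player-possessed prop for a Prop Hunt mode that handles the possessing pawn's prop controls.
; On initialization and end of possession it pushes or pops the prop input mode for locally controlled pawns.
; Input events drive the pawn's abilities by tag: cancel after a minimum time since creation, make noise, and toggle placement when allowed.
; The health update events are overridden with empty bodies.
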(event EventDoPawnRelatedInitialization
  (bind _pawnattachedto (Variables|Default|GetPawnAttachedTo))
  ;[1 1 1] DoPawnRelatedInitialization - push input component
  (|CallParentFunction)
  (if (Pawn|IsLocallyControlled _pawnattachedto)
    (PropHunt|PushPossessedPropInputMode (Variables|Default|GetFortRiftBubblePossessedPropInput) (Fortnite|GetFortPlayerControllerfromActor _pawnattachedto))))

(event EventEndPropPossession
  (bind _pawnattachedto (Variables|Default|GetPawnAttachedTo))
  ;[1 1 1] EndPropPossession - pop input component
  (Utilities|FlowControl|Sequence
    (:then_0
      (if (Pawn|IsLocallyControlled _pawnattachedto)
        (PropHunt|PopPossessedPropInputMode (Variables|Default|GetFortRiftBubblePossessedPropInput) (Fortnite|GetFortPlayerControllerfromActor _pawnattachedto))))
    (:then_1
      (|CallParentFunction))))

(event OnPossessedPropCancel(FortRiftBubblePossessedPropInput)
  (bind _pawnattachedto (Variables|Default|GetPawnAttachedTo))
  ;[1 1 1] OnPossessedPropCancel - cancel possessed prop ability
  (if (> (Actor|GetGameTimeSinceCreation) (Variables|Default|GetMinPropPossessionTime))
    (Character|UnCrouch _pawnattachedto)
    (bind _asfort_ability_system_component (Utilities|Casting|CastToFortAbilitySystemComponent (Ability|GetAbilitySystemComponent _pawnattachedto)))
    (Tags|CancelAbilitiesWithTags _asfort_ability_system_component (Variables|Default|GetPossessedPropTag))))

(event OnPossessedPropMakeNoise(FortRiftBubblePossessedPropInput)
  ;[1 1 1] OnPossessedPropMakeNoise - activate "make noise" ability
  (Abilities|TryActivateAbilitiesbyTag (Ability|GetAbilitySystemComponent (Variables|Default|GetPawnAttachedTo)) (Variables|Default|GetPossessedPropMakeNoiseTag) true))

(event OnPossessedPropPlace(FortRiftBubblePossessedPropInput)
  ;[1 1 1] OnPossessedPropPlace - Toggle prop placement mode
  (if (Variables|Default|GetIsPlaced)
    (bind _asfort_ability_system_component (Utilities|Casting|CastToFortAbilitySystemComponent (Ability|GetAbilitySystemComponent (Variables|Default|GetPawnAttachedTo))))
    (Tags|CancelAbilitiesWithTags _asfort_ability_system_component (Variables|Default|GetPossessedPropPlacedTag))
    (elif (Variables|Default|GetAllowPropPlacementAbility)
      (Abilities|TryActivateAbilitiesbyTag (Ability|GetAbilitySystemComponent (Variables|Default|GetPawnAttachedTo)) (Variables|Default|GetPossessedPropPlacedTag) true))))

(event EventUpdateHealth)

(event EventSetMaxHealth
  (params (NewMaxHealth :type Float%20%28double-precision%29)))

(fn ConstructionScript
  (|CallParentFunction))
